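; Base hero stand loot container for PistolCanary Training, built on the tiered chest parent and resettable after looting.
; On playlist ready, a band-aid fix reapplies the loot tier group, clears the searched flag and makes it invulnerable.
; A copied cosmetics override hides the item mesh instead of destroying it, so loot reset can restore visuals.
; Looting makes the mesh ignore a trace channel so bounced loot stays collectable; reset restores blocking.
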(event EventBeginPlay
  (|CallParentFunction)
  (Interaction|BindEventtoOnInteract (EventDispatchers|CreateEvent :!event Custom|OnInteract_Event))
  ; Band-aid solution @Kyle.Cavasos
  (Class|FortAsyncActionCurrentPlaylistReady|PlaylistDataReadyAsync
    (:OnReady
      (Container|SetSearchLootTierGroup (Variables|Container|GetSearchLootTierGroup))
      (Variables|Container|SetbAlreadySearched false)
      (Variables|BuildingActor|SetbIsInvulnerable true))))

(event Custom|OnInteract_Event
  (params (SelfActor :type Building%20Actor%20Object%20Reference) (InteractingPawn :type Fort%20Pawn%20Object%20Reference)))

(event EventHideChestOpenableCosmetics
  ; This section is copied from the parent class. Override it in order not to destroy "Item Mesh" component.
  (Rendering|SetVisibility (Variables|Default|GetCustom%20Smoke%20Sheet) false false)
  (CallFunction|SetLightVisibility false)
  (bind _returnvalue (Container|GetSearchingPawn))
  (Utilities|IsValid _returnvalue
    (:Is%20Valid
      (if (Pawn|IsLocallyViewed _returnvalue)
        (bind _chest_ambient_sound (Variables|Default|GetChest_Ambient_Sound)
          (:then
            (Audio|Components|Audio|FadeOut _chest_ambient_sound 0.5 0.0 "Linear" :!class /Script/Engine.AudioComponent)
            (Variables|Default|SetbOpened true))
          (:else))
        (else
          (CallFunction|AudioonOpened)))
      (CallFunction|StopAudiovisualIndicator)
      (CallFunction|CleanupWiggleMIDs)
      (Components|Timeline|Stop (Variables|Tiered_Chest_Athena_Parent|GetMobileSelectedTL))
      (CallFunction|SetVisibleMobileInteractIcon false)
      (Rendering|SetRenderCustomDepth (Variables|Debug|GetStaticMeshComponent) false)
      (if (Variables|Default|GetbOpened)
        (SoundLibrary|DisableEventsforActor (WorldSubsystems|GetSoundLibrarySubsystem) self)
        (Development|SetHiddeninGame (Variables|Default|GetHeroItemMesh) true true)))
    (:Is%20Not%20Valid
      _)))

(event EventOnLoot
  (|CallParentFunction)
  (Collision|SetCollisionResponsetoChannel (Variables|Debug|GetStaticMeshComponent) "ECC_GameTraceChannel1" "ECR_Ignore") ; Fix issue where loot that bounces back into the container cannot be looted
)

(event EventOnLootRepeat
  (|CallParentFunction)
  (Collision|SetCollisionResponsetoChannel (Variables|Debug|GetStaticMeshComponent) "ECC_GameTraceChannel1" "ECR_Ignore"))

(event EventOnLootReset
  (Variables|Container|SetbAlreadySearched false)
  (CallFunction|UnhideChestOpenableCosmetics))

(event Custom|UnhideChestOpenableCosmetics
  (Development|SetHiddeninGame (Variables|Default|GetItemMesh) false false)
  (Rendering|SetVisibility (Variables|Default|GetFX%20Glow%20&%20Smoke%20Sheet) true false)
  (CallFunction|SetLightVisibility true)
  (Collision|SetCollisionResponsetoChannel (Variables|Debug|GetStaticMeshComponent) "ECC_GameTraceChannel1" "ECR_Block") ; Fix issue where loot that bounces back into the container cannot be looted
)

(fn ConstructionScript
  (|CallParentFunction))
